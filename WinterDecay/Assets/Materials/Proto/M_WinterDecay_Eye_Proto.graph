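# Prototype masked, lit eye material for WinterDecay skeletal meshes, drawing a stylized procedural eye.
# Jittered, bump-offset parallax UVs drive concentric color rings and atan2-based polygon shapes, with the central shape morphed through Scalar_Time_of_Day.
# A normal map reflects the atmospheric light as an inner highlight, and Fresnel tints grazing angles.
# MF_WinterDecay_Intro adds emissive and supplies the opacity mask.
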Material /WinterDecay/Assets/Materials/Proto/M_WinterDecay_Eye_Proto
graph {
 properties {
  bAutomaticallySetUsageInEditor=false
  bUseMaterialAttributes=true
  bUsedWithMeshDeformer=true
  bUsedWithSkeletalMesh=true
  blendMode=BLEND_Masked
  usedShadingModels=Default Lit
 }
 nodes [
  n1 = Add(A=TextureCoordinate() {uTiling=0, vTiling=0}, B=Constant2Vector())
  n2 = Multiply(A=ScalarParameter() {defaultValue=3, parameterName=JitterAmount}, B=Time())
  n3 = Panner(Coordinate=@n1, Time=@n2) {speedX=-0.1, speedY=0.2}
  n4 = Subtract(A=Multiply(A=Frac(None=@n3)) {constB=2})
  n5 = Saturate(None=SmoothStep(Value=Abs(None=@n4)) {constMax=0.4, constMin=0.6})
  n6 = Multiply(A=ComponentMask(None=@n5) {r=true}) {constB=0.2}
  n7 = Multiply(A=ComponentMask(None=@n5) {g=true}) {constB=0.4}
  n8 = Panner(Coordinate=@n1, Time=@n2) {speedX=0.12, speedY=-0.05}
  n9 = Subtract(A=Multiply(A=Frac(None=@n8)) {constB=2})
  n10 = Saturate(None=SmoothStep(Value=Abs(None=@n9)) {constMax=0.4, constMin=0.6})
  n11 = Multiply(A=ComponentMask(None=@n10) {r=true}) {constB=0.2}
  n12 = Multiply(A=ComponentMask(None=@n10) {g=true}) {constB=0.4}
  n13 = Multiply(A=AppendVector(A=Add(A=@n6, B=@n7), B=Add(A=@n11, B=@n12))) {constB=0.3}
  NamedRerouteDeclaration(None=ConstantBiasScale(None=Clamp(None=@n13)) {bias=-0.5, scale=0.15}) {name=Jitter UV, nodeColor=<rgba 1 0 0.7734 1>}
  n14 = TextureCoordinate()
  n15 = ScalarParameter() {defaultValue=-5, parameterName=DepthFront}
  n16 = ScalarParameter() {defaultValue=-15, parameterName=DepthSide}
  n17 = CameraVectorWS()
  n18 = VertexNormalWS()
  n19 = Power(Base=OneMinus(None=DotProduct(A=@n17, B=@n18)))
  n20 = Add(A=@n14) {constB=-0.5}
  n21 = Constant2Vector()
  n22 = BumpOffset(
   Coordinate=Add(A=@n14) {constB=0.05}
   Height=Multiply(
    A=LinearInterpolate(A=@n15, B=@n16, Alpha=@n19) {constA=-10, constB=-15}
    B=SphereMask(A=@n20, B=@n21) {attenuationRadius=0.5, hardnessPercent=0}
   )
  )
  n23 = Add(A=Add(A=NamedRerouteUsage(), B=@n22)) {constB=-0.5}
  n24 = Multiply(A=Step(Y=Length(None=@n23) {is=class MaterialExpressionLength}) {constX=0.95}, B=Constant3Vector() {constant=<rgba 1 0.237 0 1>})
  n25 = Length(None=@n23)
  n26 = Step(Y=@n25) {constX=0.6}
  n27 = Step(Y=@n25) {constX=0.4}
  n28 = Constant3Vector() {constant=<rgba 1 0.331 0.0654 1>}
  n29 = Step(Y=@n25) {constX=0.25}
  n30 = Multiply(A=@n29, B=Constant3Vector() {constant=<rgba 1 0.7085 0.0219 1>})
  n31 = Constant() {r=1}
  n32 = Add(A=NamedRerouteUsage(), B=@n22)
  n33 = Add(A=Multiply(A=@n32, B=Constant2Vector() {g=0.75, r=4}), B=Constant2Vector() {g=0.125, r=-1.5})
  n34 = Constant4Vector() {constant=<rgba 0.5 0.2 0.75 1>}
  n35 = MaterialFunctionCall(Value_Morning=@n34.R, Value_Day=@n34.R, Value_Evening=@n34.R, Value_Night=@n34.R) {materialFunction=&Scalar_Time_of_Day}
  n36 = Constant() {r=-0.08}
  n37 = Subtract(
   A=MaterialFunctionCall(UVs=LinearInterpolate(A=@n33, B=@n32, Alpha=@n35.Result), "Rotation Angle (0-1)"=LinearInterpolate(B=@n36, Alpha=@n35.Result) {constB=3}) {materialFunction=&CustomRotator}.Rotated Values
   B=Constant2Vector() {g=0.5, r=0.5}
  )
  n38 = Divide(A=Divide(A=@n31) {constB=10}, B=Length(None=@n37)) {constA=0.06}
  n39 = Constant() {r=5}
  n40 = Constant2Vector() {g=2, r=2}
  n41 = Multiply(A=Multiply(A=@n37, B=@n40), B=@n40) {desc=*UVScale}
  n42 = Length(None=@n41)
  n43 = MaterialFunctionCall(Multiplier=Constant() {r=2}) {materialFunction=&Pi}
  n44 = Divide(A=Arctangent2!(Y=ComponentMask(None=@n41) {r=true}, X=ComponentMask(None=@n41) {g=true}), B=@n43.Result)
  n45 = Constant() {r=0.5}
  n46 = LinearInterpolate(Alpha=@n35.Result) {constA=4, constB=3}
  n47 = Frac(None=Multiply(A=Add(A=@n44, B=@n45), B=@n46)) {desc=t.x}
  n48 = Constant() {desc=randMix, r=0.5}
  n49 = Divide(A=Multiply(A=Subtract(A=@n47, B=@n48), B=@n43.Result), B=@n46)
  n50 = SmoothStep(
   Min=Subtract(A=@n39, B=ScalarParameter() {defaultValue=10, parameterName=EdgeSoftness})
   Max=@n39
   Value=Divide(A=Divide(B=@n42) {desc=t.y}, B=Cosine(None=@n49) {desc=Period:2pi, period=6.283})
  ) {constMax=5, constMin=4.5}
  n51 = Saturate(None=Step(Y=@n31, X=Multiply(A=@n38, B=@n50)) {constX=0.1, constY=0.6})
  n52 = Multiply(A=@n51, B=Constant3Vector())
  n53 = LinearInterpolate(
   A=Multiply(A=@n26, B=Constant3Vector() {constant=<rgba 0.8177 0.1387 0.159 1>})
   B=LinearInterpolate(A=Multiply(A=@n27, B=@n28), B=LinearInterpolate(A=@n30, B=@n52, Alpha=@n51), Alpha=@n29)
   Alpha=@n27
  )
  n54 = Divide(A=Constant() {r=4}) {constB=10}
  n55 = Divide(A=@n54)
  n56 = Subtract(A=@n55) {constB=0.016}
  n57 = Subtract(
   A=MaterialFunctionCall(UVs=@n32, "Rotation Angle (0-1)"=Constant()) {materialFunction=&CustomRotator}.Rotated Values
   B=Constant2Vector() {g=0.5, r=0.5}
  )
  n58 = Length(None=@n57)
  n59 = Constant() {r=5}
  n60 = Subtract(A=@n59, B=Constant() {r=15})
  n61 = Constant2Vector() {g=2, r=2}
  n62 = Multiply(A=Multiply(A=@n57, B=@n61), B=@n61) {desc=*UVScale}
  n63 = Arctangent2(Y=ComponentMask(None=@n62) {r=true}, X=ComponentMask(None=@n62) {g=true})
  n64 = MaterialFunctionCall(Multiplier=Constant() {r=2}) {materialFunction=&Pi}
  n65 = Constant() {r=3}
  n66 = Multiply(A=Add(A=Divide(A=@n63, B=@n64.Result), B=Constant() {r=0.5}), B=@n65)
  n67 = Multiply(A=Subtract(A=Frac(None=@n66) {desc=t.x}, B=Constant() {desc=randMix, r=0.5}), B=@n64.Result)
  n68 = Divide(A=Divide(B=Length(None=@n62)) {desc=t.y}, B=Cosine(None=Divide(A=@n67, B=@n65)) {desc=Period:2pi, period=6.283})
  n69 = Multiply(
   A=Divide(A=@n54, B=@n58) {constA=0.06}
   B=Multiply(A=Divide(B=@n58) {constA=0.02}, B=SmoothStep(Min=@n60, Max=@n59, Value=@n68) {constMax=5, constMin=4.5})
  )
  n70 = Saturate(None=Subtract(A=Step(Y=@n56, X=@n69) {constX=0.1, constY=0.2}, B=Step(Y=@n55, X=@n69) {constX=0.1, constY=0.2}))
  n71 = Constant3Vector() {constant=<rgba 0.3854 0.0949 0.2917 1>}
  n72 = MaterialFunctionCall(UVs=@n22, Texture Scale=ScalarParameter() {defaultValue=10, group=EyeColor, parameterName=IrisRadius}) {materialFunction=&ScaleUVsByCenter}
  n73 = Step(X=Length(None=Add(A=@n72.UVs) {constB=-0.5})) {constX=0.2, constY=0.2}
  n74 = OneMinus(None=MaterialFunctionCall(Radius=Constant() {r=0.5}, Density=Constant() {r=11.234}) {materialFunction=&RadialGradientExponential}.RadialGradientExponential)
  n75 = LinearInterpolate(
   A=LinearInterpolate(A=LinearInterpolate(A=@n24, B=@n53, Alpha=@n26), B=Multiply(A=@n70, B=@n71), Alpha=@n70)
   Alpha=Saturate(None=Add(A=@n73, B=@n74))
  ) {constB=0}
  n76 = ScalarParameter() {defaultValue=2.5, group=EyeShading, parameterName=Inner Highlight Specular}
  n77 = AtmosphericLightVector()
  n78 = ReflectionVectorWS(CustomWorldNormal=Transform(None=TextureSample(UVs=@n72.UVs) {samplerType=SAMPLERTYPE_Normal, texture=/Game/Global/Textures/Default/Eyes/Sphere_EYE_NORMALS}.RGB))
  n79 = Power(
   Base=Saturate(None=DotProduct(A=@n77, B=@n78))
   Exp=ScalarParameter() {defaultValue=10, group=EyeShading, parameterName=Inner HighlightRoughness}
  )
  n80 = Add(A=@n75, B=Multiply(A=@n75, B=Multiply(A=@n76, B=@n79)))
  n81 = VectorParameter() {desc=Need for energy conservative lighting, group=EyeShading, parameterName=GrazingTint_B, defaultValue=<rgba 0.0469 0.0183 0.0221 1>}
  n82 = Fresnel(ExponentIn=ScalarParameter() {defaultValue=0.7, group=EyeShading, parameterName=GrazingShadeExponent}) {baseReflectFraction=0}
  n83 = GetMaterialAttributes(None=MakeMaterialAttributes!(
    BaseColor=LinearInterpolate(A=@n80, B=@n81.RGB, Alpha=@n82) {constB=0.25}
    Metallic
    Specular=ScalarParameter() {defaultValue=0.35, group=EyeShading, parameterName=SurfaceSpecular}
    Roughness=ScalarParameter() {defaultValue=0.1, group=EyeShading, parameterName=SurfaceRoughness}
    Anisotropy
    EmissiveColor
    Opacity
    OpacityMask
    Normal
    Tangent
    WorldPositionOffset
    SubsurfaceColor
    ClearCoat
    ClearCoatRoughness
    AmbientOcclusion
    Refraction
    CustomizedUVs_0
    CustomizedUVs_1
    CustomizedUVs_2
    CustomizedUVs_3
    CustomizedUVs_4
    CustomizedUVs_5
    CustomizedUVs_6
    CustomizedUVs_7
    PixelDepthOffset
    ShadingModel
    Displacement
   )) {attributeGetTypes [#1]}
  n84 = MaterialFunctionCall() {materialFunction=/WinterDecay/Assets/Materials/MF_WinterDecay_Intro}
  SetMaterialAttributes!(MaterialAttributes=@n83.MaterialAttributes, Emissive Color=Add(A=@n83.Emissive Color, B=@n84.Emissive), Opacity Mask=@n84.OpacityMask) {attributeSetTypes [#1, #2]}
 ]
}
